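# Animation blueprint for the female medium Dieselpunk body cosmetic that follows the parent mesh's pose and adds corrective posing to its shoulder cloth.
# Copy Pose From Mesh reads the attached parent, then feeds right and left Pose Drivers chained into the output.
# Each driver uses RBF interpolation of upperarm orientation to drive only that side's dyn_shoulder_cloth bone from Pistol_ADS_RBF_PoseAsset.
# Both sides share identical pose targets, and the event graph holds no logic.
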AnimBlueprint /BRCosmetics/Characters/Player/Female/Medium/Bodies/F_MED_Dieselpunk_01/Meshes/F_MED_Dieselpunk_01_AnimBP
target_skeleton=/Game/Characters/Player/Female/Medium/Bodies/F_MED_Dieselpunk_01/Meshes/F_MED_Dieselpunk_01_Skeleton
graphs [
 AnimGraph {nodes [
   n1 = Pose Driver - Source: upperarm_r!(SourcePose:Pose Link Structure=Copy Pose From Mesh() {bUseAttachedParent=true}) -> (Pose:Pose Link Structure) {
    is=title AnimGraphNode_PoseDriver
    poseAsset=&/Pistol_ADS_RBF_PoseAsset
    onlyDriveBones [{boneName=dyn_shoulder_cloth_r}]
    poseTargets [
     {
      bApplyCustomCurve=false
      bIsHidden=false
      distanceMethod=DefaultMethod
      drivenName=Pose_0
      functionType=DefaultFunction
      targetScale=0.8104
      boneTransforms [{targetRotation=<rot -58.1985 2.4027 1.3901>, targetTranslation=<-9.8658 1.2136 0.4104>}]
      customCurve {defaultValue=3.4028234663852886e+38, postInfinityExtrap=RCCE_Constant, preInfinityExtrap=RCCE_Constant, keys []}
      targetRotation=<rot 0 0 0>
     }
     {
      bApplyCustomCurve=false
      bIsHidden=false
      distanceMethod=DefaultMethod
      drivenName=Pose_1
      functionType=DefaultFunction
      targetScale=1
      boneTransforms [{targetRotation=<rot -2.3066 102.5669 -65.3659>, targetTranslation=<-9.8658 1.2136 0.4104>}]
      customCurve {defaultValue=3.4028234663852886e+38, postInfinityExtrap=RCCE_Constant, preInfinityExtrap=RCCE_Constant, keys []}
      targetRotation=<rot 0 0 0>
     }
     {
      bApplyCustomCurve=false
      bIsHidden=false
      distanceMethod=DefaultMethod
      drivenName=Pose_2
      functionType=DefaultFunction
      targetScale=0.308
      boneTransforms [{targetRotation=<rot 30.4347 15.8578 -71.0256>, targetTranslation=<-9.8658 1.2136 0.4104>}]
      customCurve {defaultValue=3.4028234663852886e+38, postInfinityExtrap=RCCE_Constant, preInfinityExtrap=RCCE_Constant, keys []}
      targetRotation=<rot 0 0 0>
     }
     {
      bApplyCustomCurve=false
      bIsHidden=false
      distanceMethod=DefaultMethod
      drivenName=Pose_3
      functionType=DefaultFunction
      targetScale=0.308
      boneTransforms [{targetRotation=<rot 12.9936 -6.7632 -99.3037>, targetTranslation=<-9.8658 1.2136 0.4104>}]
      customCurve {defaultValue=3.4028234663852886e+38, postInfinityExtrap=RCCE_Constant, preInfinityExtrap=RCCE_Constant, keys []}
      targetRotation=<rot 0 0 0>
     }
    ]
    rBFParams {radius=44.1671}
    sourceBones [{boneName=upperarm_r}]
   }
   n2 = Pose Driver - Source: upperarm_l!(SourcePose:Pose Link Structure=@n1) -> (Pose:Pose Link Structure) {
    is=title AnimGraphNode_PoseDriver
    poseAsset=&Pistol_ADS_RBF_PoseAsset
    onlyDriveBones [{boneName=dyn_shoulder_cloth_l}]
    poseTargets [
     {
      bApplyCustomCurve=false
      bIsHidden=false
      distanceMethod=DefaultMethod
      drivenName=Pose_0
      functionType=DefaultFunction
      targetScale=0.8104
      boneTransforms [{targetRotation=<rot -58.1985 2.4027 1.3901>, targetTranslation=<-9.8658 1.2136 0.4104>}]
      customCurve {defaultValue=3.4028234663852886e+38, postInfinityExtrap=RCCE_Constant, preInfinityExtrap=RCCE_Constant, keys []}
      targetRotation=<rot 0 0 0>
     }
     {
      bApplyCustomCurve=false
      bIsHidden=false
      distanceMethod=DefaultMethod
      drivenName=Pose_1
      functionType=DefaultFunction
      targetScale=1
      boneTransforms [{targetRotation=<rot -2.3066 102.5669 -65.3659>, targetTranslation=<-9.8658 1.2136 0.4104>}]
      customCurve {defaultValue=3.4028234663852886e+38, postInfinityExtrap=RCCE_Constant, preInfinityExtrap=RCCE_Constant, keys []}
      targetRotation=<rot 0 0 0>
     }
     {
      bApplyCustomCurve=false
      bIsHidden=false
      distanceMethod=DefaultMethod
      drivenName=Pose_2
      functionType=DefaultFunction
      targetScale=0.308
      boneTransforms [{targetRotation=<rot 30.4347 15.8578 -71.0256>, targetTranslation=<-9.8658 1.2136 0.4104>}]
      customCurve {defaultValue=3.4028234663852886e+38, postInfinityExtrap=RCCE_Constant, preInfinityExtrap=RCCE_Constant, keys []}
      targetRotation=<rot 0 0 0>
     }
     {
      bApplyCustomCurve=false
      bIsHidden=false
      distanceMethod=DefaultMethod
      drivenName=Pose_3
      functionType=DefaultFunction
      targetScale=0.308
      boneTransforms [{targetRotation=<rot 12.9936 -6.7632 -99.3037>, targetTranslation=<-9.8658 1.2136 0.4104>}]
      customCurve {defaultValue=3.4028234663852886e+38, postInfinityExtrap=RCCE_Constant, preInfinityExtrap=RCCE_Constant, keys []}
      targetRotation=<rot 0 0 0>
     }
    ]
    rBFParams {radius=44.1671}
    sourceBones [{boneName=upperarm_l}]
   }
   Output Pose(Result=@n2) {layerGroup=None}
  ]}
 EventGraph {nodes [Event BlueprintUpdateAnimation()]}
]
variables []
