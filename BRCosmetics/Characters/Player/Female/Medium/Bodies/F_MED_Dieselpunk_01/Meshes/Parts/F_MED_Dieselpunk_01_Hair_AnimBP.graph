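# Animation Blueprint for the hair part of the female Dieselpunk 01 player body, adding secondary physics motion to the hair.
# It copies the pose from the attached parent mesh, then chains six Anim Dynamics nodes, one per dyn_hair bone.
# Each uses low gravity, angular limits and a spherical limit on neck_01.
# Pawn 2D speed drives their shared alpha, strengthening the effect with movement; the event graph has no logic.
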AnimBlueprint /BRCosmetics/Characters/Player/Female/Medium/Bodies/F_MED_Dieselpunk_01/Meshes/Parts/F_MED_Dieselpunk_01_Hair_AnimBP
target_skeleton=&/F_MED_Dieselpunk_01_Hair_Skeleton
graphs [
 AnimGraph {nodes [
   n1 = float * float(A=MapRangeClamped(Value=Get PawnSpeed2D(), InRangeA=0, InRangeB=450, OutRangeA=0.4, OutRangeB=0.8), B=1)
   n2 = Anim Dynamics - Bone: dyn_hair_04(ComponentPose=Local To Component(LocalPose=Copy Pose From Mesh() {bUseAttachedParent=true}), Alpha=@n1, ExternalForce="0,0,0") {
    angularDampingOverride=0.98
    bGravityOverrideInSimSpace=true
    bOverrideAngularDamping=true
    gravityScale=0.1
    numSolverIterationsPostUpdate=4
    numSolverIterationsPreUpdate=16
    alphaBoolBlend {blendOption=Linear}
    boundBone {boneName=dyn_hair_04}
    componentLinearAccScale=<1 1 1>
    componentLinearVelScale=<1 1 1>
    physicsBodyDefinitions [{
      sphereCollisionRadius=0
      boundBone {boneName=dyn_hair_04}
      boxExtents=<20 20 10>
      constraintSetup {twistAxis=AxisZ, angularLimitsMax=<20 20 0>, angularLimitsMin=<-20 -20 0>, angularTarget=<0 0 0>}
      localJointOffset=<0 0 -5>
     }]
    sphericalLimits [{limitRadius=6, drivingBone {boneName=neck_01}, sphereLocalOffset=<0 0 0>}]
   }
   n3 = Anim Dynamics - Bone: dyn_hair_01(ComponentPose=@n2, Alpha=@n1) {
    angularDampingOverride=0.98
    bGravityOverrideInSimSpace=true
    bOverrideAngularDamping=true
    gravityScale=0.1
    numSolverIterationsPostUpdate=4
    numSolverIterationsPreUpdate=16
    alphaBoolBlend {blendOption=Linear}
    boundBone {boneName=dyn_hair_01}
    componentLinearAccScale=<1 1 1>
    componentLinearVelScale=<1 1 1>
    physicsBodyDefinitions [{
      sphereCollisionRadius=0
      boundBone {boneName=dyn_hair_01}
      boxExtents=<20 20 10>
      constraintSetup {twistAxis=AxisZ, angularLimitsMax=<20 20 0>, angularLimitsMin=<-20 -20 0>, angularTarget=<0 0 0>}
      localJointOffset=<0 0 -5>
     }]
    sphericalLimits [{limitRadius=6, drivingBone {boneName=neck_01}, sphereLocalOffset=<0 0 0>}]
   }
   n4 = Anim Dynamics - Bone: dyn_hair_02(ComponentPose=@n3, Alpha=@n1) {
    angularDampingOverride=0.98
    bGravityOverrideInSimSpace=true
    bOverrideAngularDamping=true
    gravityScale=0.1
    numSolverIterationsPostUpdate=4
    numSolverIterationsPreUpdate=16
    alphaBoolBlend {blendOption=Linear}
    boundBone {boneName=dyn_hair_02}
    componentLinearAccScale=<1 1 1>
    componentLinearVelScale=<1 1 1>
    physicsBodyDefinitions [{
      sphereCollisionRadius=0
      boundBone {boneName=dyn_hair_02}
      boxExtents=<20 20 10>
      constraintSetup {twistAxis=AxisZ, angularLimitsMax=<20 20 0>, angularLimitsMin=<-20 -20 0>, angularTarget=<0 0 0>}
      localJointOffset=<0 0 -5>
     }]
    sphericalLimits [{limitRadius=6, drivingBone {boneName=neck_01}, sphereLocalOffset=<0 0 0>}]
   }
   n5 = Anim Dynamics - Bone: dyn_hair_05(ComponentPose=@n4, Alpha=@n1, ExternalForce="0,0,0") {
    angularDampingOverride=0.98
    bGravityOverrideInSimSpace=true
    bOverrideAngularDamping=true
    gravityScale=0.1
    numSolverIterationsPostUpdate=4
    numSolverIterationsPreUpdate=16
    alphaBoolBlend {blendOption=Linear}
    boundBone {boneName=dyn_hair_05}
    componentLinearAccScale=<1 1 1>
    componentLinearVelScale=<1 1 1>
    physicsBodyDefinitions [{
      sphereCollisionRadius=0
      boundBone {boneName=dyn_hair_05}
      boxExtents=<20 20 10>
      constraintSetup {twistAxis=AxisZ, angularLimitsMax=<20 20 0>, angularLimitsMin=<-20 -20 0>, angularTarget=<0 0 0>}
      localJointOffset=<0 0 -5>
     }]
    sphericalLimits [{limitRadius=6, drivingBone {boneName=neck_01}, sphereLocalOffset=<0 0 0>}]
   }
   n6 = Anim Dynamics - Bone: dyn_hair_03(ComponentPose=@n5, Alpha=@n1, ExternalForce="0,0,0") {
    angularDampingOverride=0.98
    bGravityOverrideInSimSpace=true
    bOverrideAngularDamping=true
    gravityScale=0.1
    numSolverIterationsPostUpdate=4
    numSolverIterationsPreUpdate=16
    alphaBoolBlend {blendOption=Linear}
    boundBone {boneName=dyn_hair_03}
    componentLinearAccScale=<1 1 1>
    componentLinearVelScale=<1 1 1>
    physicsBodyDefinitions [{
      sphereCollisionRadius=0
      boundBone {boneName=dyn_hair_03}
      boxExtents=<20 20 10>
      constraintSetup {twistAxis=AxisZ, angularLimitsMax=<20 20 0>, angularLimitsMin=<-20 -20 0>, angularTarget=<0 0 0>}
      localJointOffset=<0 0 -5>
     }]
    sphericalLimits [{limitRadius=6, drivingBone {boneName=neck_01}, sphereLocalOffset=<0 0 0>}]
   }
   n7 = Anim Dynamics - Bone: dyn_hair_06(ComponentPose=@n6, Alpha=@n1, ExternalForce="0,0,0") {
    angularDampingOverride=0.98
    bGravityOverrideInSimSpace=true
    bOverrideAngularDamping=true
    gravityScale=0.1
    numSolverIterationsPostUpdate=4
    numSolverIterationsPreUpdate=16
    alphaBoolBlend {blendOption=Linear}
    boundBone {boneName=dyn_hair_06}
    componentLinearAccScale=<1 1 1>
    componentLinearVelScale=<1 1 1>
    physicsBodyDefinitions [{
      sphereCollisionRadius=0
      boundBone {boneName=dyn_hair_06}
      boxExtents=<20 20 10>
      constraintSetup {twistAxis=AxisZ, angularLimitsMax=<20 0 0>, angularLimitsMin=<-20 0 0>, angularTarget=<0 0 0>}
      localJointOffset=<0 0 -5>
     }]
    sphericalLimits [{limitRadius=6, drivingBone {boneName=neck_01}, sphereLocalOffset=<0 0 0>}]
   }
   Output Pose(Result=Component To Local(ComponentPose=@n7)) {layerGroup=None}
  ]}
 EventGraph {nodes [Event BlueprintUpdateAnimation()]}
]
variables []
